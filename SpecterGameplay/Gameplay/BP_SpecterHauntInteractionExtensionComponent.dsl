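; Haunt interaction extension component that tracks, on clients, how many specters haunt its owner and drives the looping wisp effect.
; On non-authority BeginPlay it seeds the count from existing haunters, then client haunt start and stop events increment or decrement it.
; Count changes add a local looping GameplayCue when rising from zero and remove it at zero.
; Each change is broadcast through the global GameplayEventRouter with the owning player pawn.
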(event EventBeginPlay
  (bind _returnvalue (CallFunction|GetHaunterCount))
  ; Set initial haunt state. e.g. player may come into relevancy with existing haunts.
  (Networking|SwitchHasAuthority
    (:Remote
      (if (> _returnvalue 0)
        (CallFunction|OnNumHauntsChanged _returnvalue)))))

(event EventBP_OnHauntStartClient
  (CallFunction|OnNumHauntsChanged (+ (Variables|Default|GetCurrentNumHaunts) 1)))

(event EventBP_OnHauntStopClient
  (CallFunction|OnNumHauntsChanged (- (Variables|Default|GetCurrentNumHaunts) 1)))

(fn OnNumHauntsChanged
  (params (InNewNumHaunts :type Integer))
  (if (!= InNewNumHaunts (Variables|Default|GetCurrentNumHaunts))
    (Utilities|FlowControl|Sequence
      (:then_0
        (if (> (Variables|GetInNewNumHaunts) 0)
          ; If we have wisps, but no GCNL, add it.
          (if (<= (Variables|Default|GetCurrentNumHaunts) 0)
            (Projectile|AddGameplayCue%28Looping%29Local (Components|GetOwner) (Variables|Default|GetGameplay%20Cue%20Tag%20Wisps) (Ability|GameplayCue|MakeGameplayCueParameters 0.0)))
          (else
            ; If we've reached zero wisps, remove the GCNL
            (Projectile|RemoveGameplayCue%28Looping%29Local (Components|GetOwner) (Variables|Default|GetGameplay%20Cue%20Tag%20Wisps) (Ability|GameplayCue|MakeGameplayCueParameters 0.0)))))
      (:then_1
        (Variables|Default|SetCurrentNumHaunts (Variables|GetInNewNumHaunts))
        ; Send router event of our new count.
        (bind _asfort_player_pawn_athena (Utilities|Casting|CastToFortPlayerPawnAthena (Components|GetOwner)))
        (GameplayEventRouter|BroadcastEvent (GameplayEventRouter|GetGlobalEventRouter) (Utilities|Struct|MakeEventSpecterNumHauntsChanged _asfort_player_pawn_athena (Variables|Default|GetCurrentNumHaunts)) _asfort_player_pawn_athena)))))
