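; Specter haunting component Blueprint that handles the gameplay effect and audio feedback when a haunt starts and ends.
; On the server, it applies the haunting GameplayEffect to the owner's Fort ability system, stores the handle, and removes it when haunting ends.
; On the client, start plays a 2D sound and deactivates the specter-mode control bus mix; end stops that sound, plays a stop sound, and reactivates the mix.
; The server attach and detach events are empty.
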(event EventBP_OnHauntStartServer
  (bind _returnvalue (Ability|GetFortAbilitySystemComponent (Components|GetOwner)))
  (Utilities|IsValid _returnvalue
    (:Is%20Valid
      (bind _returnvalue_2 (GameplayEffects|ApplyGameplayEffectToTarget :self _returnvalue :GameplayEffectClass (Variables|Default|GetHauntingEffect) :Target _returnvalue :Level 0.0))
      (Variables|Default|SetHauntingGEHandle _returnvalue_2))))

(event EventBP_OnHauntStartClient
  (bind _returnvalue (Audio|SpawnSound2D "/SpecterGameplay/Audio/Gameplay/MSS_SpecterPawn_Haunt_Start.MSS_SpecterPawn_Haunt_Start"))
  (Audio|Modulation|DeactivateControlBusMix "/SpecterGameplay/Audio/Data/CBM_InSpecterMode.CBM_InSpecterMode"))

(event EventBP_OnHauntEndServer
  (bind _returnvalue (Ability|GetFortAbilitySystemComponent (Components|GetOwner)))
  (Utilities|IsValid _returnvalue
    (:Is%20Valid
      (GameplayEffects|RemoveActiveGameplayEffect _returnvalue (Variables|Default|GetHauntingGEHandle) -1))))

(event EventBP_OnHauntEndClient
  (Audio|Components|Audio|Stop (stale EventBP_OnHauntStartClient _returnvalue))
  (Audio|SpawnSound2D "/SpecterGameplay/Audio/Gameplay/MSS_SpecterPawn_Haunt_Stop.MSS_SpecterPawn_Haunt_Stop" 1.0 0.7)
  (Audio|Modulation|ActivateControlBusMix "/SpecterGameplay/Audio/Data/CBM_InSpecterMode.CBM_InSpecterMode"))

(event EventBP_AttachSpecterToPawnServer
  (params (PawnToHaunt :type Fort%20Player%20Pawn%20Athena%20Object%20Reference) (SpecterPawn :type Specter%20Pawn%20Object%20Reference)))

(event EventBP_DetachSpecterFromPawnServer
  (params (PawnToHaunt :type Fort%20Player%20Pawn%20Athena%20Object%20Reference) (SpecterPawn :type Specter%20Pawn%20Object%20Reference)))
