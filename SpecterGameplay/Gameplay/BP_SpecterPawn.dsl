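; Player ghost (Specter) pawn that applies the specter-mode gameplay effect on authority, with overlap-only capsule collision that still blocks world static.
; Spawn and despawn play Niagara poofs, a tail and sounds; async tag waits toggle a sprint cue effect with TacticalSprint.
; Haunting hides mesh and tail and disables capsule collision; on exit a capsule trace corrects underground positions.
; Jump, movement and movement-mode events feed grounded, moving and jump animation flags.
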(event Collision|EventActorBeginOverlap
  (params (OtherActor :type Actor%20Object%20Reference)))

(event EventTick
  (params (DeltaSeconds :type Float%20%28single-precision%29)))

(event EventBeginPlay
  (bind _returnvalue (Ability|GetAbilitySystemComponent 0))
  (bind _capsulecomponent (Variables|Character|GetCapsuleComponent))
  (bind _mesh (Variables|Character|GetMesh))
  ;[0.15 0.15 0.15 0.5 30] Begin Play
  (Utilities|FlowControl|Sequence
    (:then_0
      (Networking|SwitchHasAuthority
        (:Authority
          (Utilities|IsValid _returnvalue :!label _label
            (:Is%20Valid
              (CallFunction|StartSpecterGameplayEffects _returnvalue))
            (:Is%20Not%20Valid
              (Utilities|FlowControl|Delay 0.2)
              (goto _label))))))
    (:then_1
      (|CallParentFunction))
    (:then_2
      ;;[1 1 1] Disable collisions
      (Collision|SetCollisionResponsetoAllChannels _capsulecomponent "ECR_Overlap")
      (Collision|SetCollisionResponsetoChannel _capsulecomponent "ECC_WorldStatic" "ECR_Block")
      (if (ScalableFloat|GetValueasBool (Variables|Default|GetHF_SetCapsuleToCameraObjectType))
        (Collision|SetCollisionObjectType _capsulecomponent "ECC_Camera"))
      (Collision|SetCollisionEnabled (Variables|Character|GetMesh) "NoCollision")
      (Fortnite|SetCanbeDamaged self))
    (:then_3
      (bind _returnvalue_2 (Niagara|SpawnSystemAttached (Variables|Default|GetNS_Ghost_SpawnPoof) _mesh "Root" "0, 0, 0" "0, 0, 0" "KeepRelativeOffset" true))
      (Variables|Default|SetNS_Ghost_SpawnPoof_Handle _returnvalue_2)
      (bind _returnvalue_3 (Niagara|SpawnSystemAttached "/SpecterGameplay/Gameplay/Effects/Ghost/NS_Ghost_Tail.NS_Ghost_Tail" _mesh "Root" "0, 0, 0" "0, 0, 0" "KeepRelativeOffset" true))
      (Variables|Default|SetNS_Ghost_Tail_Handle _returnvalue_3)
      (Audio|SpawnSoundAttached "/SpecterGameplay/Audio/Gameplay/MSS_SpecterPawn_Appear_1P.MSS_SpecterPawn_Appear_1P" _mesh))
    (:then_4
      ;;[1 1 1 0.5] Animation Events
      (Pawn|AssignOnPawnJumped (delegate Custom|GhostJumped))
      (EventDispatchers|BindEventtoOnPawnLanded (delegate Custom|Ghostlanded))
      (Character|BindEventtoOnCharacterMovementUpdated (delegate Custom|UpdateGhostMoving)))))

(event Custom|GhostJumped
  ;[0.15 0.15 0.15 0.5 30] Begin Play
  ;;[1 1 1 0.5] Animation Events
  (Variables|Default|SetABP_JumpTriggered true)
  (Utilities|FlowControl|DelayUntilNextFrame)
  (Variables|Default|SetABP_JumpTriggered false))

(event Custom|UpdateGhostMoving
  (params (DeltaSeconds :type Float%20%28single-precision%29) (OldLocation :type Vector) (OldVelocity :type Vector))
  ;[0.15 0.15 0.15 0.5 30] Begin Play
  ;;[1 1 1 0.5] Animation Events
  (Variables|Default|SetABP_IsMoving (> (Math|Vector|VectorLengthXYSquared OldVelocity) 0.0001)))

(event Custom|Ghostlanded
  (params (Hit :type Hit%20Result%20Structure%20%28by%20ref%29))
  ;[0.15 0.15 0.15 0.5 30] Begin Play
  ;;[1 1 1 0.5] Animation Events
  (Variables|Default|SetABP_JumpTriggered false))

(event EventOnMovementModeChanged
  (params (PrevMovementMode :type EMovementMode%20Enum :default "MOVE_None") (NewMovementMode :type EMovementMode%20Enum :default "MOVE_None") (PrevCustomMode :type Byte) (NewCustomMode :type Byte))
  ;[0.15 0.15 0.15 0.5 30] Begin Play
  ;;[1 1 1 0.5] Animation Events
  (CallFunction|UpdateAnimIsGrounded NewMovementMode NewCustomMode))

(event Custom|StartSpecterGameplayEffects
  (params (InAbilitySystemComp :type Ability%20System%20Component%20Object%20Reference))
  (bind _returnvalue (Ability|GetAbilitySystemComponent 0))
  (bind _returnvalue_2 (GameplayEffects|ApplyGameplayEffectToSelf :self InAbilitySystemComp :GameplayEffectClass "/SpecterGameplay/Abilities/GE_InSpecterMode.GE_InSpecterMode_C" :Level 0.0))
  (Variables|Default|SetGE_Handle _returnvalue_2)
  ;[0.15 0.15 0.15 0.5 30] Apply and Remove ghost sprint vfx
  ;; Wait for Sprint tag to get added, then apply GE_SpecterSprint
  (bind _asynctaskproxy (Ability|Tasks|WaitGameplayTagAddToActor 0 "(TagName=\"Gameplay.Action.Player.TacticalSprint\")")
    (:then
      (Variables|Default|SetAsyncTaskStartSprint _asynctaskproxy) ; For EndAction OnEndPlay
      ;; Wait for Sprint tag to get removed, then remove all GE_SpecterSprint
      (bind _asynctaskproxy_2 (Ability|Tasks|WaitGameplayTagRemoveFromActor 0 "(TagName=\"Gameplay.Action.Player.TacticalSprint\")")
        (:then
          (Variables|Default|SetAsyncTaskEndSprint _asynctaskproxy_2) ; For EndAction OnEndPlay
        )
        (:Removed
          (Utilities|IsValid _returnvalue
            (:Is%20Valid
              (GameplayEffects|RemoveActiveGameplayEffectbySourceEffect _returnvalue "/SpecterGameplay/Abilities/GE_Specter_GrantSprintCue.GE_Specter_GrantSprintCue_C" _returnvalue -1))))))
    (:Added
      (Utilities|IsValid _returnvalue
        (:Is%20Valid
          (GameplayEffects|ApplyGameplayEffectToSelf :self _returnvalue :GameplayEffectClass "/SpecterGameplay/Abilities/GE_Specter_GrantSprintCue.GE_Specter_GrantSprintCue_C" :Level 0.0))))) ; Waits until tag is applied to actor
)

(event EventEndPlay
  (params (EndPlayReason :type EEndPlayReason%20Enum :default "Destroyed"))
  (bind _returnvalue (Transformation|GetActorLocation))
  (bind _capsulecomponent (Variables|Character|GetCapsuleComponent))
  (bind _returnvalue_2 (Components|Capsule|GetScaledCapsuleHalfHeight _capsulecomponent))
  (bind _returnvalue_3 (Ability|GetAbilitySystemComponent 0))
  (bind _returnvalue_4 (Utilities|Operators|Add _returnvalue (Math|Vector|MakeVector 0.0 0.0 _returnvalue_2)))
  ;[0.15 0.15 0.15 0.5 30] End Play
  (Utilities|FlowControl|Sequence
    (:then_0
      (Utilities|IsValid _returnvalue_3
        (:Is%20Valid
          (GameplayEffects|RemoveActiveGameplayEffect _returnvalue_3 (Variables|Default|GetGE_Handle) -1))))
    (:then_1
      (bind _returnvalue_5 (Niagara|SpawnSystematLocation (Variables|Default|GetNS_Ghost_DespawnPoof) _returnvalue_4))
      (Variables|Default|SetNS_Ghost_DespawnPoof_Handle _returnvalue_5)
      (Components|Activation|Deactivate (Variables|Default|GetNS_Ghost_Tail_Handle))
      (Audio|SpawnSoundatLocation "/SpecterGameplay/Audio/Gameplay/MSS_SpecterPawn_Disappear_3P.MSS_SpecterPawn_Disappear_3P" _returnvalue_4))
    (:then_2
      (Ability|Async|EndAction (Variables|Default|GetAsyncTaskStartSprint)))
    (:then_3
      (|CallParentFunction EndPlayReason))))

(event EventBP_OnHauntStartServer
  ;[0.15 0.15 0.15 0.5 30] On Haunt
  (Rendering|SetVisibility (Variables|Character|GetMesh) false false)
  (Collision|SetCollisionEnabled (Variables|Character|GetCapsuleComponent) "NoCollision"))

(event EventBP_OnHauntStopServer
  ;[0.15 0.15 0.15 0.5 30] On Haunt
  (Rendering|SetVisibility (Variables|Character|GetMesh) true false)
  ;; see if we're in the ground, adjust if we are
  (bind (_outhit _returnvalue) (Collision|CapsuleTraceByChannel (Transformation|GetWorldLocation (Variables|Character|GetCapsuleComponent)) (Transformation|GetWorldLocation (Variables|Character|GetCapsuleComponent)) (Variables|/Script/Engine.CapsuleComponent|GetCapsuleRadius (Variables|Character|GetCapsuleComponent)) (Variables|/Script/Engine.CapsuleComponent|GetCapsuleHalfHeight (Variables|Character|GetCapsuleComponent)) "TraceTypeQuery10"))
  (if _returnvalue
    (Fortnite|FortShippingLog self "Ghost is underground, adjusting respawn location" true)
    (bind _output_get (Variables|Default|SetLeaveHauntLocation (Transformation|GetActorLocation)))
    (Class|FortAthenaMutatorUtilities|AdjustRespawnLocationforGroundHeight _output_get (* (Variables|/Script/Engine.CapsuleComponent|GetCapsuleHalfHeight (Variables|Character|GetCapsuleComponent)) 2.0) self)
    (Transformation|SetActorLocation (Variables|Default|GetLeaveHauntLocation) false true))
  (Collision|SetCollisionEnabled (Variables|Character|GetCapsuleComponent) "QueryAndPhysics"))

(event EventBP_OnHauntStartClient
  ;[0.15 0.15 0.15 0.5 30] On Haunt
  (CallFunction|PlayHauntForceFeedback)
  (Rendering|SetVisibility (Variables|Default|GetNS_Ghost_Tail_Handle) false false)
  (Default|CallHauntingClient true))

(event EventBP_OnHauntStopClient
  ;[0.15 0.15 0.15 0.5 30] On Haunt
  (CallFunction|PlayHauntForceFeedback)
  (Rendering|SetVisibility (Variables|Default|GetNS_Ghost_Tail_Handle) true false)
  (Default|CallHauntingClient false))

(event Blocking|EventRemoveTagsToBlockWithMelee
  (params (GameplayTagsToBlock :type Gameplay%20Tag%20Container%20Structure%20%28by%20ref%29)))

(fn GetAnimData
  (return (Variables|Default|GetABP_IsGrounded) (Variables|Default|GetABP_IsMoving) (Variables|Default|GetABP_JumpTriggered)))

(fn GetTagsToBlockWithMelee
  (return 0))

(fn GetZoomedOutCameraFramingSettings
  (return 0))

(fn PlayHauntForceFeedback
  (bind _self self)
  (bind _returnvalue (Fortnite|GetFortPlayerControllerfromActor _self))
  (Utilities|IsValid _returnvalue
    (:Is%20Valid
      (bind _hauntforcefeedback (Variables|Default|GetHauntForceFeedback)
        (:then
          (Game|Feedback|LocalPlayForceFeedback _returnvalue _hauntforcefeedback false "Specter.Haunt")
          (Development|PrintString "play haunt force feedback"))
        (:else)))))

(fn UpdateAnimIsGrounded
  (params (NewMovementMode :type EMovementMode%20Enum :default "MOVE_None") (NewCustomMovementMode :type Byte))
  (switch Utilities|FlowControl|Switch|SwitchonEMovementMode NewMovementMode
    (:MOVE_Walking
      _)
    (:MOVE_NavWalking
      _)
    (:MOVE_Falling
      (Variables|Default|SetABP_IsGrounded false :!label _label))
    (:MOVE_Custom
      (switch Utilities|FlowControl|Switch|SwitchonEFortCustomMovement (Utilities|Enum|BytetoEnumEFortCustomMovement NewCustomMovementMode)
        (:Driving
          _)
        (:Passenger
          _)
        (:Parachuting
          (goto _label))
        (:Skydiving
          (goto _label))
        (:SkydiveFollowing
          (goto _label))
        (:Grinding
          _)
        (:SurfaceSwimming
          _)
        (:DeimosPreFall
          (goto _label))
        (:Sliding
          _)
        (:Skating
          _)
        (:SwingingGrounded
          _))))
  (Variables|Default|SetABP_IsGrounded true))

(fn ConstructionScript)
